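; Junk Rift falling object: an accelerating projectile that damages what it hits, then spawns a lodged prop.
; A given or weighted-random MeshInt selects mesh, transform, player-damage effects and prop.
; The server traces ground and water, damages queued buildings on a timer, launches overlapped pawns and credits quests.
; Reaching ground or an unbreakable blocker triggers area damage; blockers suppress the prop, and clients run their own distance check for cosmetics.
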(event Custom|SelectTransform
  (bind _meshint (Variables|MeshSelect|GetMeshInt))
  (bind _mapmeshtransforms (Variables|MeshSelect|GetMapMeshTransforms))
  (bind (_value _returnvalue) (Utilities|Map|Find _mapmeshtransforms _meshint))
  ;[0.2501 1 0.11] Select and set the mesh's offset transform based on the user defined MeshInt.
  (if _returnvalue
    (bind _output_get (Variables|MeshSelect|SetMeshOffsetTransform _value))
    (Transformation|SetRelativeTransform (Variables|Default|GetFallingMesh) _output_get false false)
    (else
      ;;[1 0 0 1 12] Error if MeshInt not found in MapMeshTransforms
      (Development|PrintString "BGA_Athena_JollyRascal_MainDrop::SelectTransform - Failed to find a transform with the specified MeshInt in MapMeshTransforms." true true "(R=1.000000,G=0.000000,B=0.000000,A=1.000000)"))))

(event Custom|SelectGEs
  (bind _meshint (Variables|MeshSelect|GetMeshInt))
  (bind _map_gefinalsoft (Variables|MeshSelect|GetMap_GEFinalSoft))
  (bind (_value _returnvalue) (Utilities|Map|Find _map_gefinalsoft _meshint))
  (bind _meshint_2 (Variables|MeshSelect|GetMeshInt))
  (bind _map_geprimarysoft (Variables|MeshSelect|GetMap_GEPrimarySoft))
  (bind (_value_2 _returnvalue_2) (Utilities|Map|Find _map_geprimarysoft _meshint_2))
  ;[0.09237 1 0.1206] Select and load all GEs that will need to be used.
  (Utilities|FlowControl|Sequence
    (:then_0
      ;;[0.7664 1 0.1309] Select, load, and assign GE_PlayerDamage based on the user defined MeshInt.
      (if _returnvalue_2
        (bind _class (Utilities|AsyncLoadClassAsset _value_2)
          (:Completed
            (bind _asgameplay_effect (Utilities|Casting|CastToGameplayEffectClass _class))
            (Variables|GEs|SetGE_PlayerDamage _asgameplay_effect)))
        (else
          ;;;[1 0 0 1 12] Error if MeshInt not found in Map_GEPrimarySoft
          (Development|PrintString "BGA_Athena_JollyRascal_MainDrop::SelectGEs - Failed to find a GE with the specified MeshInt in Map_GEPrimarySoft." true true "(R=1.000000,G=0.000000,B=0.000000,A=1.000000)"))))
    (:then_1
      ;;[0.9051 1 0.1971] Select, load, and assign GE_FinalPlayerDamage based on the user defined MeshInt.
      (if _returnvalue
        (bind _class_2 (Utilities|AsyncLoadClassAsset _value)
          (:Completed
            (bind _asgameplay_effect_2 (Utilities|Casting|CastToGameplayEffectClass _class_2))
            (Variables|GEs|SetGE_FinalPlayerDamage _asgameplay_effect_2)))
        (else
          ;;;[1 0 0 1 12] Error if MeshInt not found in Map_GEFinalSoft
          (Development|PrintString "BGA_Athena_JollyRascal_MainDrop::SelectGEs - Failed to find a GE with the specified MeshInt in Map_GEFinalSoft." true true "(R=1.000000,G=0.000000,B=0.000000,A=1.000000)"))))))

(event Custom|FallFasterOverTime
  ;[1 1 1] Speed Up Over Time
  (bind (_direction _vel) (Timeline|VelOverTime PlayFromStart)
    (:Update
      (Variables|/Script/Engine.MovementComponent|SetVelocity_X (Math|Vector|MakeVector 0.0 0.0 _vel) (Variables|Default|GetFortProjectileMovement)))))

(event Custom|SelectMeshAndMaterial
  ;[0.05805 1 0.1878] Select, load, and apply the selected static mesh and material according to the user defined MeshInt.
  (bind _object (Utilities|AsyncLoadAsset (Utilities|Array|Get(acopy) (Variables|MeshSelect|GetSoftMeshs) (select (Utilities|Array|IsValidIndex (Variables|MeshSelect|GetSoftMeshs) (Variables|MeshSelect|GetMeshInt)) (Variables|MeshSelect|GetMeshInt) (Utilities|Array|LastIndex (Variables|MeshSelect|GetSoftMeshs)))))
    (:Completed
      (bind _asstatic_mesh (Utilities|Casting|CastToStaticMesh _object))
      (bind _output_get (Variables|MeshSelect|SetSelectedMesh _asstatic_mesh))
      (Components|StaticMesh|SetStaticMesh (Variables|Default|GetFallingMesh) _output_get)
      ;;[0.9434 1 0.1661 1 12] Select/load the material after the mesh has been applied so it doesn't get overridden by the mesh's material.
      (CallFunction|SelectMaterial))))

(event Custom|HitBlazer
  ;[1 1 1] Hit Blazer Force Stop Early
  (Utilities|FlowControl|DoOnce false
    (:Completed
      (Development|PrintString "HitBlazer")
      (Variables|UnbreakableCheck|SetBlazerBlocked true)
      (CallFunction|FinalHit))))

(event Custom|SelectMaterial)

(event Custom|ClientTickDistCheck
  ;[1 1 1] ClientDistanceCheck
  (CallFunction|DistanceCheckClient))

(event Custom|Traceforworld
  (bind _returnvalue (Transformation|GetActorLocation))
  (bind _linetraceoffset (Variables|GroundCheck|GetLineTraceOffset))
  (bind _returnvalue_2 (+ _returnvalue _linetraceoffset))
  ;[1 1 1] Get trace for the world to find where we should stop
  (Utilities|FlowControl|Sequence
    (:then_0
      (bind (_outhit _returnvalue_3) (Collision|LineTraceByChannel _returnvalue _returnvalue_2 "TraceTypeQuery7"))
      (if _returnvalue_3
        (Variables|GroundCheck|SetGroundLocation (.Location (Collision|BreakHitResult _outhit)))
        (Variables|GroundCheck|SetFoundHit true)
        (else
          ;[1 1 1 0.5] Find Solid Ground if the NonDestructible trace fails.
          (bind (_outhits _returnvalue_4) (Collision|MultiLineTraceForObjects _returnvalue _returnvalue_2 (Utilities|Array|MakeArray "ObjectTypeQuery1")))
          (if _returnvalue_4
            (bind _array_element (Utilities|Array|ForEachLoopwithBreak _outhits)
              (:LoopBody
                (if (== (Utilities|GetClass (.HitComponent (Collision|BreakHitResult _array_element))) "/Script/Landscape.LandscapeHeightfieldCollisionComponent")
                  (Variables|GroundCheck|SetGroundLocation (.Location (Collision|BreakHitResult _array_element)))
                  (Variables|GroundCheck|SetFoundHit true))))))))
    (:then_1
      (bind (_outhit_2 _returnvalue_5) (Collision|LineTraceByChannel _returnvalue _returnvalue_2 "TraceTypeQuery12"))
      (if _returnvalue_5
        (bind _aswater_body (Utilities|Casting|CastToWaterBody (.HitActor (Collision|BreakHitResult _outhit_2))))
        (Variables|Default|SetWater%20Impact%20Point (WaterBody|GetWaterSurfaceInfoatLocation (Variables|/Script/Water.WaterBody|GetWaterBodyComponent _aswater_body) (.ImpactPoint (Collision|BreakHitResult _outhit_2)) true))
        (Variables|Default|SetWater%20Surface%20Normal (.OutWaterSurfaceNormal (WaterBody|GetWaterSurfaceInfoatLocation (Variables|/Script/Water.WaterBody|GetWaterBodyComponent _aswater_body) (.ImpactPoint (Collision|BreakHitResult _outhit_2)) true)))
        (Variables|Default|SetWater%20Depth (.OutWaterDepth (WaterBody|GetWaterSurfaceInfoatLocation (Variables|/Script/Water.WaterBody|GetWaterBodyComponent _aswater_body) (.ImpactPoint (Collision|BreakHitResult _outhit_2)) true)))
        (Variables|Default|SetFound%20Water true)))))

(event EventBeginPlay
  (bind (_returnvalue_x _returnvalue_y _returnvalue_z) (Math|Vector|BreakVector (Transformation|GetActorLocation)))
  (bind _returnvalue (Actor|GetOwner))
  (bind _returnvalue_2 (Fortnite|GetFortPlayerControllerfromActor _returnvalue))
  ;[1 1 1] Begin Play - Save Out Values and start overlap Timer
  (Utilities|FlowControl|Sequence
    (:then_0
      ;[1 1 1] Initialize the random prop chosen if it wasn't set specifically on spawn
      (if (< (Variables|MeshSelect|GetMeshInt) 0)
        (bind _selectedindex (CallFunction|WeightedRandomChoice))
        (bind _output_get (Variables|MeshSelect|SetMeshInt _selectedindex))
        (Development|PrintString (Utilities|String|Append "Initializing Junk Rift to randomly chosen int: " (Utilities|String|ToString(Integer) _output_get)))))
    (:then_1
      ;[1 1 1] Check if player is on junk rift quest
      (Networking|SwitchHasAuthority
        (:Authority
          (if (Utilities|IsValid _returnvalue_2 :!class /Script/Engine.KismetSystemLibrary)
            (bind (_questisvalid _questcompleted _objectivecompleted) (Default|AthenaCheckQuestandObjectiveCompletion _returnvalue_2 0 (Variables|Quest|GetQuestBackendName)))
            (bind (_bisassistingobjective _bisvalidplayercontroller) (Class|AthenaQuestFunctionLibrary|AthenaPartyAssistCheck _returnvalue_2 (Variables|Quest|GetQuestBackendName)))
            (Variables|Quest|SetSendQuestEvent (or (and _bisassistingobjective _bisvalidplayercontroller) (Math|Boolean|ANDBoolean _questisvalid (not _questcompleted) (not _objectivecompleted))))))))
    (:then_2
      (CallFunction|SetStartingValues)
      (CallFunction|SelectMeshandMaterial)
      (CallFunction|SelectTransform)
      (Ability|ExecuteGameplayCueWithParamsReplicated (Variables|Cues|GetGC_Spawn) (Ability|GameplayCue|MakeGameplayCueParameters 0.0 (Math|Conversions|ToFloat(Integer) (select (== (Variables|MeshSelect|GetMeshInt) 3) 4 (Variables|MeshSelect|GetMeshInt)))))
      (CallFunction|SelectGEs)
      (CallFunction|FallFasterOverTime)
      (Variables|Default|SetInitialHeight _returnvalue_z)
      (if (Networking|IsServer)
        (CallFunction|Traceforworld)
        (Utilities|Array|AddUnique (Variables|Destruction|GetOverlapBuildingActorsToIgnore) self)
        (bind _returnvalue_3 (Utilities|Time|SetTimerbyEvent (EventDispatchers|CreateEvent :!event Custom|DamageBuildingsOnTick) (Variables|Destruction|GetTimeTraceTick) true))
        (Variables|Destruction|SetTimer_Trace _returnvalue_3)
        (else
          (bind _returnvalue_4 (Utilities|Time|SetTimerbyEvent (delegate Custom|ClientTickDistCheck) (Variables|Destruction|GetTimeTraceTick) true))
          (Variables|GroundCheck|SetTimer_TraceClient _returnvalue_4)))
      (|CallParentFunction))))

(event Custom|DelayedFinalHit
  ;[1 1 1] Selay Before Final Hit If Called This way
  (Utilities|FlowControl|Delay (Variables|Default|GetDelay_DelayedFinalHit))
  (CallFunction|FinalHit))

(event Custom|SpawnProp
  (bind _meshint (Variables|MeshSelect|GetMeshInt))
  (bind _map_prop (Variables|MeshSelect|GetMap_Prop))
  (bind (_value _returnvalue) (Utilities|Map|Find _map_prop _meshint))
  (bind (_returnvalue_location _returnvalue_rotation _returnvalue_scale) (Math|Transform|BreakTransform (Transformation|GetWorldTransform (Variables|Default|GetFallingMesh))))
  (bind (_returnvalue_location_x _returnvalue_location_y) (Math|Vector|BreakVector _returnvalue_location))
  ;[1 1 1] Spawn Prop
  (if (Variables|UnbreakableCheck|GetBlazerBlocked)
    ;;[1 1 1] Hide Mesh if Hit Blazer
    (Variables|Default|SetMeshHidden true)
    (elif _returnvalue
      (bind _returnvalue_2 (Game|SpawnBuildingActorfromClass _value (Math|Transform|MakeTransform (Math|Vector|MakeVector _returnvalue_location_x _returnvalue_location_y (CallFunction|GetAcceptableLodgedinGroundHeight)) _returnvalue_rotation _returnvalue_scale) (Game|GetInstigator) false 0 0 (Game|GetInstigator)))
      (bind _output_get (Variables|Default|SetSpawnedProp _returnvalue_2))
      (Ability|GameplayEffect|AddLooseGameplayTags _output_get (GameplayTags|MakeLiteralGameplayTagContainer "(GameplayTags=((TagName=\"Trait.Weight.VeryHeavy\")))")) ; Prevent the object from getting picked up by Tornado
      ;[1 1 1] Apply Bonus HP to Prop and stop destruction from building placement
      (bind _returnvalue_3 (Utilities|MakeMyFortContext))
      (GameplayEffects|ApplyGameplayEffectToTarget (Ability|GetAbilitySystemComponent self) (Variables|GEs|GetGE_PropBonusHP) (Ability|GetAbilitySystemComponent (Variables|Default|GetSpawnedProp)) 0.0 _returnvalue_3)
      (Variables|/Script/FortniteGame.BuildingActor|SetbDestroyOnPlayerBuildingPlacement false (Variables|Default|GetSpawnedProp)))))

(event Custom|HideAndKill
  ;[1 1 1] Play GC, Hide Actor and Kill
  (Game|Components|ProjectileMovement|StopSimulating (Variables|Default|GetFortProjectileMovement) (Variables|Kill|GetNullHit))
  (Components|Movement|StopMovementImmediately (Variables|Default|GetFortProjectileMovement))
  (Ability|ExecuteGameplayCueWithParamsReplicated (Variables|Cues|GetGC_Destroy) (Ability|GameplayCue|MakeGameplayCueParameters :NormalizedMagnitude 0.0 :RawMagnitude (Math|Conversions|ToFloat(Integer) (select (== (Variables|MeshSelect|GetMeshInt) 3) 4 (Variables|MeshSelect|GetMeshInt))) :Location (CallFunction|GetFinalLocation) :Normal "0,0,0" :GameplayEffectLevel 1 :AbilityLevel 1 :bReplicateLocationWhenUsingMinimalRepProxy false))
  (Components|Activation|Deactivate (Variables|Default|GetRotatingMovement))
  (Networking|ForceNetUpdate)
  (Fortnite|ShowSoundIndicatorLocalforTeam :Location (Transformation|GetActorLocation) :MaxAudibleDistance 4e+03 :IndicatorType "Gunshot" :Team (Team|GetActorTeam self) :Affiliations (Utilities|Array|MakeArray "Hostile" "Neutral") :Tint "(R=0.000000,G=0.000000,B=0.000000,A=0.000000)" :AttenuationMaxRangeMultiplier 1.0 :AttenuationIntensityMultiplier 1.0)
  (Collision|SetActorEnableCollision false)
  (Components|Activation|Deactivate (Variables|Default|GetFallFX))
  (Audio|Components|Audio|FadeOut (Variables|Default|GetAudio) 0.25 0.0 "Linear" :!class /Script/Engine.AudioComponent)
  (if (Networking|IsServer)
    (Actor|SetLifeSpan 2.0)
    (elif (Variables|UnbreakableCheck|GetBlazerBlocked)
      _
      (else
        (Components|Timeline|PlayfromStart (Variables|BGA_Figment_JollyRascal_MainDrop|GetSyncLocation))
        (Rendering|Material|SetScalarParameterValueonMaterials (Variables|Default|GetFallingMesh) "WobbleStartTime" (Utilities|Time|GetGameTimeinSeconds))))))

(event Custom|DamageBuildingsOnTick
  (bind _buildingstodamage (Variables|Destruction|GetBuildingsToDamage))
  (bind _output (Utilities|Array|Get(acopy) _buildingstodamage))
  ;[0.1858 1 0.9502] Deal damage to overlapped buildings on a timer that is set in BeginPlay.
  (CallFunction|DistanceCheck)
  (if (not (Variables|Kill|GetPendingKill))
    (Variables|Quest|SetQuestCount 0)
    (Utilities|FlowControl|Sequence
      (:then_0
        ;;[0.09493 1 0] Collect any newly overlapped actors and queue them to be damaged.
        (bind _outactors (Collision|SphereOverlapActors (+ (Transformation|GetActorLocation) (Variables|Destruction|GetOverlapOffset)) (Variables|Destruction|GetOverlapBuildingRadius) (Variables|Destruction|GetOverlapBuildingObjectType) "/Script/FortniteGame.BuildingActor" (Variables|Destruction|GetOverlapBuildingActorsToIgnore)))
        (Utilities|Array|AppendArray (Variables|Destruction|GetBuildingsToDamage) _outactors)
        (Utilities|Array|AppendArray (Variables|Destruction|GetOverlapBuildingActorsToIgnore) _outactors))
      (:then_1
        ;;[0.05306 0.3743 1] Attempt to deal damage to the maximum number of queued buildings allowed per trace.
        (Utilities|FlowControl|ForLoopwithBreak 0 (Variables|Destruction|GetMaxBuildingsToDamagePerTrace)
          (:LoopBody
            (if (Utilities|Array|IsValidIndex (Variables|Destruction|GetBuildingsToDamage))
              _
              (else
                _))
            (Utilities|FlowControl|Sequence
              (:then_0
                ;;;[0.1041 1 0.6523] Attempt to deal damage to the building that was hit.
                (bind _asbuilding_gameplay_actor (Utilities|Casting|CastToBuildingGameplayActor _output)
                  (:then
                    ;;;;[0.8646 0.4201 0] If the mesh hits something that's unbreakable, trigger the final hit.
                    (if (GameplayTags|TagContainer|CompareTagAssetInterfacetoOtherTags (Variables|UnbreakableCheck|GetT_UnbreakableBGA) _asbuilding_gameplay_actor)
                      (CallFunction|HitBlazer)))
                  (:CastFailed
                    (bind _returnvalue (CallFunction|CheckforCanbeDamaged _output))
                    (if _returnvalue
                      (bind _returnvalue_2 (Utilities|MakeMyFortContext))
                      (GameplayEffects|ApplyGameplayEffectToTarget (Ability|GetAbilitySystemComponent self) (Variables|GEs|GetGE_EnvironmentalDamage) (Ability|GetAbilitySystemComponent _output) 0.0 _returnvalue_2)
                      ;;;;[0 0 1] Update the quest count if this actor meets the requirements.
                      (bind _bdidactorhavetags (Class|AthenaQuestFunctionLibrary|AthenaCheckActorHasTags _output "(GameplayTags=((TagName=\"Building.Resource\")))"))
                      (if (and _bdidactorhavetags (Variables|Quest|GetSendQuestEvent))
                        (Math|Integer|IncrementInt (Variables|Quest|GetQuestCount)))))))
              (:then_1
                ;;;[1 0.5861 0.08121] Remove the building from the queue of buildings to damage.
                (Utilities|Array|RemoveIndex (Variables|Destruction|GetBuildingsToDamage)))))))
      (:then_2
        ;;[0 0 1] Broadcast the quest event if any buildings were damaged.
        (if (> (Variables|Quest|GetQuestCount) 0)
          (Quest|SendComplexCustomStatEvent (Quests|GetQuestManager (Fortnite|GetFortPlayerControllerfromActor (Actor|GetOwner)) "Invalid") 0 "(GameplayTags=)" (Variables|Quest|GetQuestTargetTag) (Variables|Quest|GetQuestCount)))))))

(event OnComponentBeginOverlap(PawnCollision)
  (params (OverlappedComponent :type Primitive%20Component%20Object%20Reference) (OtherActor :type Actor%20Object%20Reference) (OtherComp :type Primitive%20Component%20Object%20Reference) (OtherBodyIndex :type Integer) (bFromSweep :type Boolean) (SweepResult :type Hit%20Result%20Structure%20%28by%20ref%29))
  (bind _returnvalue (Transformation|GetActorLocation))
  (bind _knockmag (Variables|PawnHits|GetKnockMag))
  (bind _knockaddedz (Variables|PawnHits|GetKnockAddedZ))
  ;[1 1 1] On overlap with pawns apply impulse and/or damage
  (if (Networking|IsServer)
    (if (not (Utilities|Array|ContainsItem (Variables|PawnHits|GetHitPawns) OtherActor))
      (bind _asfort_vehicle_interface (Utilities|Casting|CastToFortVehicleInterface OtherActor)
        (:then
          (bind _returnvalue_2 (Utilities|MakeMyFortContext))
          (GameplayEffects|ApplyGameplayEffectToTarget (Ability|GetAbilitySystemComponent self) (Variables|GEs|GetGE_VehicleDamage) (Vehicle|GetVehicleAbilitySystemComponent _asfort_vehicle_interface) 0.0 _returnvalue_2)
          (Utilities|Array|AddUnique (Variables|PawnHits|GetHitPawns) OtherActor))
        (:CastFailed
          (bind _asfort_pawn (Utilities|Casting|CastToFortPawn OtherActor)
            (:then
              (bind _returnvalue_3 (CallFunction|LineTraceforBlocking _asfort_pawn true))
              (if (not _returnvalue_3)
                (Utilities|Array|AddUnique (Variables|PawnHits|GetHitPawns) _asfort_pawn)
                (bind _asfort_player_pawn (Utilities|Casting|CastToFortPlayerPawn _asfort_pawn)
                  (:then
                    (bind (_returnvalue_x _returnvalue_y) (Math|Vector|BreakVector (Math|Vector|Normalize (- (Transformation|GetActorLocation _asfort_pawn) _returnvalue))))
                    (bind (_returnvalue_x_2 _returnvalue_y_2 _returnvalue_z) (Math|Vector|BreakVector (Math|Vector|vector*float (Math|Vector|MakeVector _returnvalue_x _returnvalue_y) _knockmag)))
                    (Utilities|LaunchPlayerPawn (Utilities|Casting|CastToFortPlayerPawn (Game|GetInstigator)) _asfort_player_pawn (Math|Vector|MakeVector _returnvalue_x_2 _returnvalue_y_2 (+ _returnvalue_z _knockaddedz)) true true false true true true true))
                  (:CastFailed
                    (Character|LaunchCharacter _asfort_pawn (* (Math|Vector|MakeVector _returnvalue_x_2 _returnvalue_y_2 (+ _returnvalue_z _knockaddedz)) (Variables|Default|GetDeimosImpulseMult)) true true)))
                (if (or (not (Team|OnSameTeam _asfort_pawn self)) (Utilities|Equal(Object) _asfort_pawn (Game|GetInstigator)))
                  (if (or (GameplayTags|TagContainer|CompareTagAssetInterfacetoOtherTags (Variables|BushCheck|GetT_Bush) (Variables|SnowmanCheck|GetT_Snowman) _asfort_pawn) (.Case_1%20True (GameplayTags|TagContainer|CompareTagAssetInterfacetoOtherTags (Variables|BushCheck|GetT_Bush) (Variables|SnowmanCheck|GetT_Snowman) _asfort_pawn)))
                    ;;[1 1 1] If the player has a bush remove it instead of dealing damage
                    (for _array_element (Actor|GetAttachedActors _asfort_pawn true false)
                      (bind _asathena_player_bush (Utilities|Casting|CastToAthena_Player_Bush _array_element))
                      (Class|AthenaPlayerBush|RemoveBush _asathena_player_bush))
                    (else
                      (bind _returnvalue_4 (Utilities|MakeMyFortContext))
                      (GameplayEffects|ApplyGameplayEffectToTarget (Ability|GetAbilitySystemComponent (Game|GetInstigator)) (Variables|GEs|GetGE_PlayerDamage) (Ability|GetAbilitySystemComponent _asfort_pawn) 0.0 _returnvalue_4))))))
            (:CastFailed
              (if (Utilities|IsValid (Actor|GetComponentbyClass OtherActor "/Script/FortniteGame.FortPhysicsObjectComponent") :!class /Script/Engine.KismetSystemLibrary)
                (bind _returnvalue_5 (Utilities|MakeMyFortContext))
                (GameplayEffects|ApplyGameplayEffectToTarget (Ability|GetAbilitySystemComponent self) (Variables|GEs|GetGE_VehicleDamage) (Ability|GetAbilitySystemComponent OtherActor) 0.0 _returnvalue_5)))))))))

(event Custom|FinalHit
  (bind _returnvalue (Transformation|GetActorLocation))
  (bind _knockmag (Variables|PawnHits|GetKnockMag))
  (bind _knockaddedz (Variables|PawnHits|GetKnockAddedZ))
  ;[1 1 1] Do Final AOE Hit
  (CallFunction|HideandKill)
  (Variables|Kill|SetPendingKill true)
  (Utilities|Time|ClearandInvalidateTimerbyHandle (Variables|Destruction|GetTimer_Trace))
  (Utilities|Time|ClearandInvalidateTimerbyHandle (Variables|GroundCheck|GetTimer_TraceClient))
  (bind _outactors (Collision|SphereOverlapActors (CallFunction|GetFinalLocation) (ScalableFloat|GetValueatLevel (Variables|Rows|GetRow_FinalPawnOverlap)) (Variables|PawnHits|GetFinalHitObjectType) 0 (Variables|PawnHits|GetHitPawns)))
  (for _array_element _outactors
    (if (Networking|IsServer)
      (bind _returnvalue_2 (CallFunction|LineTraceforBlocking _array_element true))
      (if (not _returnvalue_2)
        (bind _asfort_vehicle_interface (Utilities|Casting|CastToFortVehicleInterface _array_element)
          (:then
            (bind _returnvalue_3 (Utilities|MakeMyFortContext))
            (GameplayEffects|ApplyGameplayEffectToTarget (Ability|GetAbilitySystemComponent self) (Variables|GEs|GetGE_FinalVehicleDamage) (Vehicle|GetVehicleAbilitySystemComponent _asfort_vehicle_interface) 0.0 _returnvalue_3)
            (Utilities|Array|AddUnique (Variables|PawnHits|GetHitPawns) _array_element))
          (:CastFailed
            (bind _asfort_pawn (Utilities|Casting|CastToFortPawn _array_element))
            (Utilities|Array|AddUnique (Variables|PawnHits|GetHitPawns) _asfort_pawn)
            (bind _asfort_player_pawn (Utilities|Casting|CastToFortPlayerPawn _asfort_pawn)
              (:then
                (bind (_returnvalue_x _returnvalue_y) (Math|Vector|BreakVector (Math|Vector|Normalize (- (Transformation|GetActorLocation _asfort_pawn) _returnvalue))))
                (bind (_returnvalue_x_2 _returnvalue_y_2 _returnvalue_z) (Math|Vector|BreakVector (Math|Vector|vector*float (Math|Vector|MakeVector _returnvalue_x _returnvalue_y) _knockmag)))
                (Utilities|LaunchPlayerPawn (Utilities|Casting|CastToFortPlayerPawn (Game|GetInstigator)) _asfort_player_pawn (Math|Vector|MakeVector _returnvalue_x_2 _returnvalue_y_2 (+ _returnvalue_z _knockaddedz)) true true false true true true true))
              (:CastFailed
                (Character|LaunchCharacter _asfort_pawn (* (Math|Vector|MakeVector _returnvalue_x_2 _returnvalue_y_2 (+ _returnvalue_z _knockaddedz)) (Variables|Default|GetDeimosImpulseMult)) true true) ; Launch AI
              ))
            (if (or (not (Team|OnSameTeam _asfort_pawn self)) (Utilities|Equal(Object) _asfort_pawn (Game|GetInstigator)))
              ;;[0 0 0 1 10] Makes for the cardboard and snowman to react the same as a bush instead of killing them
              (if (or (GameplayTags|TagContainer|CompareTagAssetInterfacetoOtherTags (Variables|BushCheck|GetT_Bush) (Variables|SnowmanCheck|GetT_Snowman) _asfort_pawn) (.Case_1%20True (GameplayTags|TagContainer|CompareTagAssetInterfacetoOtherTags (Variables|BushCheck|GetT_Bush) (Variables|SnowmanCheck|GetT_Snowman) _asfort_pawn)))
                ;;[1 1 1] If the player has a bush on remove it instead of dealing damage
                (for _array_element_2 (Actor|GetAttachedActors _asfort_pawn true false)
                  (bind _asathena_player_bush (Utilities|Casting|CastToAthena_Player_Bush _array_element_2))
                  (Class|AthenaPlayerBush|RemoveBush _asathena_player_bush))
                (else
                  (bind _returnvalue_4 (Utilities|MakeMyFortContext))
                  (GameplayEffects|ApplyGameplayEffectToTarget (Ability|GetAbilitySystemComponent self) (Variables|GEs|GetGE_FinalPlayerDamage) (Ability|GetAbilitySystemComponent _asfort_pawn) 0.0 _returnvalue_4)))))))))
  (CallFunction|DamageRemainingQueuedBuildings)
  (bind _outactors_2 (Collision|SphereOverlapActors (CallFunction|GetFinalLocation) (ScalableFloat|GetValueatLevel (Variables|Rows|GetRow_FinalEnvOverlap)) (Variables|Destruction|GetOverlapBuildingObjectType) "/Script/FortniteGame.BuildingActor" (Utilities|Array|MakeArray self)))
  (for _array_element_3 _outactors_2
    (bind _returnvalue_5 (CallFunction|LineTraceforBlocking _array_element_3))
    (if (not _returnvalue_5)
      (bind _asbuilding_static_mesh (Utilities|Casting|CastToBuildingSMActor _array_element_3)
        (:then
          (if (Default|CanBuildingbeDamaged _asbuilding_static_mesh)
            _))
        (:CastFailed
          _))
      (bind _returnvalue_6 (Utilities|MakeMyFortContext))
      (GameplayEffects|ApplyGameplayEffectToTarget (Ability|GetAbilitySystemComponent self) (Variables|GEs|GetGE_FinalEnvironmentalDamage) (Ability|GetAbilitySystemComponent _array_element_3) 0.0 _returnvalue_6)))
  (CallFunction|SpawnProp))

(event Custom|DamageRemainingQueuedBuildings
  ;[1 0.4871 0.0636] Damage any buildings that are still queued for damage. Called when the final hit occurs, before dealing final damage to buildings.
  (for _array_element (Variables|Destruction|GetBuildingsToDamage)
    (bind _returnvalue (CallFunction|CheckforCanbeDamaged _array_element))
    (if _returnvalue
      (bind _returnvalue_2 (Utilities|MakeMyFortContext))
      (GameplayEffects|ApplyGameplayEffectToTarget (Ability|GetAbilitySystemComponent self) (Variables|GEs|GetGE_EnvironmentalDamage) (Ability|GetAbilitySystemComponent _array_element) 0.0 _returnvalue_2)))
  (Development|PrintString (Utilities|String|ToString(Text) (Utilities|Text|FormatText "NSLOCTEXT(\"[1715142602CEF4A350A846F919382DCF]\", \"12BEE3E5460947AA7CC587A5FF989BD7\", \"QueuedBuildings: {X}  IgnoredBuildings: {Y}\")" (Utilities|Array|Length (Variables|Destruction|GetBuildingsToDamage)) (Utilities|Array|Length (Variables|Destruction|GetOverlapBuildingActorsToIgnore)))))
  (Utilities|Array|Clear (Variables|Destruction|GetBuildingsToDamage))
  (Utilities|Array|Clear (Variables|Destruction|GetOverlapBuildingActorsToIgnore)))

(fn CanBuildingBeDamaged
  (params (BuildingSM :type Building%20Static%20Mesh%20Object%20Reference))
  (return (Class|BuildingStaticMesh|CanbeDamaged BuildingSM (Building|GetOwningController self) self)))

(fn CheckForCanBeDamaged
  (params (Target :type Actor%20Object%20Reference))
  (LocalVariables|SetHitDamagable true)
  (bind _asbuilding_static_mesh (Utilities|Casting|CastToBuildingSMActor Target)
    (:then
      _)
    (:CastFailed
      _))
  (bind _data (Utilities|FlowControl|MergeExecutionwithData (Default|CanBuildingbeDamaged _asbuilding_static_mesh) (Variables|/Script/Engine.Actor|GetbCanBeDamaged Target))
    (:Then
      (if (not _data)
        (bind (_outhits _returnvalue) (Collision|MultiLineTraceByChannel (+ (Transformation|GetActorLocation) (Variables|UnbreakableCheck|GetHitUndamagableTraceOffset)) (- (Transformation|GetActorLocation) (Variables|UnbreakableCheck|GetHitUndamagableTraceOffset)) "TraceTypeQuery10" true))
        (if _returnvalue
          (bind _array_element (Utilities|Array|ForEachLoopwithBreak _outhits)
            (:LoopBody
              (Utilities|Casting|CastToBuildingSMActor (.HitActor (Collision|BreakHitResult _array_element)))
              (Variables|GroundCheck|SetGroundLocation (+ (.Location (Collision|BreakHitResult _array_element)) (Variables|UnbreakableCheck|GetHitUnbreakableGroundOffset)))
              (CallFunction|DelayedFinalHit)
              (LocalVariables|SetHitDamagable false))
            (:Completed
              _))
          (else
            _))
        (return (LocalVariables|GetHitDamagable))))))

(fn DistanceCheck
  (bind (_returnvalue_x _returnvalue_y _returnvalue_z) (Math|Vector|BreakVector (Transformation|GetActorLocation)))
  (bind (_groundlocation_x _groundlocation_y _groundlocation_z) (Math|Vector|BreakVector (Variables|GetGroundLocation_X)))
  (bind (_water_impact_point_x _water_impact_point_y _water_impact_point_z) (Math|Vector|BreakVector (Variables|GetWater%20Impact%20Point_X)))
  (if (Math|Boolean|ANDBoolean (>= _groundlocation_z _returnvalue_z) (Variables|GroundCheck|GetFoundHit) (not (Variables|Kill|GetPendingKill)))
    (CallFunction|FinalHit)
    (Development|PrintString "Final Hit")
    (elif (and (>= _water_impact_point_z _returnvalue_z) (Variables|Default|GetFound%20Water))
      (Variables|Default|SetHit%20Water 1)
      (else
        (Variables|Default|SetHit%20Water 0)))))

(fn DistanceCheckClient
  (bind (_returnvalue_x _returnvalue_y _returnvalue_z) (Math|Vector|BreakVector (Transformation|GetActorLocation)))
  (bind (_groundlocation_x _groundlocation_y _groundlocation_z) (Math|Vector|BreakVector (Variables|GetGroundLocation_X)))
  (bind (_water_impact_point_x _water_impact_point_y _water_impact_point_z) (Math|Vector|BreakVector (Variables|GetWater%20Impact%20Point_X)))
  (if (Math|Boolean|ANDBoolean (>= _groundlocation_z _returnvalue_z) (Variables|GroundCheck|GetFoundHit) (not (Variables|Kill|GetPendingKill)))
    (Variables|Kill|SetPendingKill true)
    (Utilities|Time|ClearandInvalidateTimerbyHandle (Variables|GroundCheck|GetTimer_TraceClient))
    (CallFunction|HideandKill)
    (elif (and (>= _water_impact_point_z _returnvalue_z) (Variables|Default|GetFound%20Water))
      (Utilities|FlowControl|DoOnce false
        (:Completed
          (CallFunction|HitWater))))))

(fn GetAcceptableLodgedInGroundHeight
  (bind _returnvalue_location (Math|Transform|BreakTransform (Transformation|GetWorldTransform (Variables|Default|GetFallingMesh))))
  (bind (_returnvalue_location_x _returnvalue_location_y _returnvalue_location_z) (Math|Vector|BreakVector _returnvalue_location))
  (return (Math|Float|Clamp(Float) _returnvalue_location_z (Utilities|Operators|Add (.z (Variables|GroundCheck|GetGroundLocation)) (Utilities|Select :Index (Variables|MeshSelect|GetMeshInt) :Option%200 3.5e+02 :Option%201 1.5e+02 :Option%202 2e+02 :Option%203 -45.0) 0) (Variables|Default|GetInitialHeight))))

(fn GetFinalLocation
  (return (select (Variables|UnbreakableCheck|GetBlazerBlocked) (Transformation|GetActorLocation) (Variables|GroundCheck|GetGroundLocation))))

(fn HitWater
  (bind _water_impact_point (Variables|Default|GetWater%20Impact%20Point))
  (bind _returnvalue (Niagara|SpawnSystematLocation "/Game/Effects/Niagara/Systems/Enviroment/Water/NS_WaterSplash_Large.NS_WaterSplash_Large" _water_impact_point))
  (Niagara|SetNiagaraVariable(Vector3) _returnvalue "WaterPlaneLocation" _water_impact_point)
  (Niagara|SetNiagaraVariable(Vector3) _returnvalue "WaterPlaneNormal" (Variables|Default|GetWater%20Surface%20Normal))
  (Niagara|SetNiagaraVariable(Float) _returnvalue "WaterDepth" (Variables|Default|GetWater%20Depth))
  (Audio|PlaySoundatLocation "/Game/Athena/Sounds/Weapons/Explosions/Expl_Debris_Water_Cue.Expl_Debris_Water_Cue" (Variables|Default|GetWater%20Impact%20Point))
  (Development|PrintString "I hit water!" true true "(R=0.000000,G=0.660000,B=1.000000,A=1.000000)" 5.0))

(fn LineTraceForBlocking
  (params (Target :type Actor%20Object%20Reference) (ShouldDoBlockingCast :type Boolean))
  ;[1 1 1] Do a multiline trace to see if we hit any silver blazers. If so return true. Using a tag check so that we so not pull in the silver blazer actor. Also check for can be damaged to see if there is a indestructable actor in the way.
  (LocalVariables|SetBlockedTrace false)
  (bind _outhits (Collision|MultiLineTraceForObjects (Transformation|GetActorLocation Target) (+ (Transformation|GetActorLocation) (LocalVariables|GetHeightOffsetForTrace)) (LocalVariables|GetObjectTypes) true))
  (bind _array_element (Utilities|Array|ForEachLoopwithBreak _outhits)
    (:LoopBody
      (bind _returnvalue (GameplayTags|HasMatchingGameplayTag(Message) (.HitActor (Collision|BreakHitResult _array_element)) (Variables|UnbreakableCheck|GetT_UnbreakableBGA)))
      (if _returnvalue
        (Development|PrintString "BlazerTraceBlocked")
        (LocalVariables|SetBlockedTrace true))
      (if ShouldDoBlockingCast
        (bind _asbuilding_actor (Utilities|Casting|CastToBuildingActor (.HitActor (Collision|BreakHitResult _array_element))))
        ;;[0 0 0 1 10] It would be nice to just set the Exempt classes bCanBeDamaged = true, but in particular Athena_SneakySnowmanV2_InteractSphere needs it to be false (FORT-237723 "fixed bug where hitting snowman with snowball would break interaction with snowman"). When doing a line trace for the player pawn we need this bBlockedTrace to be false (FORT-297972 Player character fails to take damage from "Junk Rift" while hiding into "Creepin' Cardboard" in Creative)
        (if (>= (Utilities|Array|FindItem (LocalVariables|GetExemptBuildingActors) (Utilities|GetClass _asbuilding_actor)) 0)
          _
          (elif (Variables|/Script/Engine.Actor|GetbCanBeDamaged _asbuilding_actor)
            _
            (else
              (Development|PrintString "CanBeDamagedBlocked"))))))
    (:Completed
      _))
  (return (LocalVariables|GetBlockedTrace)))

(fn OnRep_HitWater
  (Networking|CanExecuteCosmeticEvents
    (:True
      (switch int (Variables|Default|GetHit%20Water)
        (:1
          (CallFunction|HitWater))))))

(fn OnRep_MeshHidden
  (Development|SetHiddeninGame (Variables|Default|GetFallingMesh) (Variables|Default|GetMeshHidden) false))

(fn SetStartingValues
  (Variables|Destruction|SetOverlapBuildingRadius (ScalableFloat|GetValueatLevel (Variables|Rows|GetRow_OverlapRadiusEnvironment)))
  (Variables|PawnHits|SetKnockMag (ScalableFloat|GetValueatLevel (Variables|Rows|GetRow_KnockMag)))
  (Variables|PawnHits|SetKnockAddedZ (ScalableFloat|GetValueatLevel (Variables|Rows|GetRow_KnockAddedZ)))
  (Variables|/Script/Engine.MovementComponent|SetVelocity (Variables|Default|GetVelocity) (Variables|Default|GetFortProjectileMovement))
  (Components|Sphere|SetSphereRadius (Variables|Default|GetPawnCollision) (ScalableFloat|GetValueatLevel (Variables|Rows|GetRow_OverlapRadiusPawn)) true))

(fn ConstructionScript
  (if (not (Networking|IsServer))
    (Collision|SetCollisionEnabled (Variables|Default|GetPawnCollision) "NoCollision"))
  (|CallParentFunction))

(fn WeightedRandomChoice
  (LocalVariables|Setcalc_maxWeight 0)
  (for _array_element (Variables|MeshSelect|GetSpawnWeights)
    (LocalVariables|Setcalc_maxWeight (+ _array_element (LocalVariables|Getcalc_maxWeight))))
  (bind _output_get (LocalVariables|SetrandomRoll (Math|Random|RandomIntegerinRange 0 (- (LocalVariables|Getcalc_maxWeight) 1))))
  (LocalVariables|Setcalc_runningSum 0)
  (for (_array_element_2 _array_index) (Variables|MeshSelect|GetSpawnWeights)
    (bind _output_get_2 (LocalVariables|Setcalc_runningSum (+ _array_element_2 (LocalVariables|Getcalc_runningSum))))
    (Development|PrintString (Utilities|String|Append "Weight Index: " (Utilities|String|ToString(Integer) _array_index) "     Running Sum: " (Utilities|String|ToString(Integer) _output_get_2) "      Random Roll: " (Utilities|String|ToString(Integer) (LocalVariables|GetrandomRoll))))
    (if (< (LocalVariables|GetrandomRoll) _output_get_2)
      (Development|PrintString "Found weighted value")
      (return _array_index)))
  (Development|PrintString "FAIL to select a weighted index, returning 0 as a safe fallback")
  (return 0))
