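; Junk Rift (JollyRascal) portal spawner that drops a main actor, then scattered extra junk, before removing itself.
; After delayed portal FX, the server spawns the dropper, then randomly timed extras from a scalable-float-driven list of resource, ammo and drone boxes.
; Spawn points clamp to minigame volumes; resource type and mesh index are rolled at spawn to avoid client desync.
; Cleanup plays an end cue, shrinks the portal via timeline and sets its lifespan.
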(event Custom|SpawnMainDrop
  ;[1 1 1] Drop the main actor
  (Game|SpawnBuildingActorfromClass (Variables|Default|GetDropperClassToSpawn) (CallFunction|ClampTransformtoMinigameVolume (Math|Transform|MakeTransform (Transformation|GetActorLocation) (Math|Rotator|MakeRotator 0.0 0.0 (Math|Random|RandomFloatinRange 0.0 3.6e+02)) "1.000000,1.000000,1.000000")) (Game|GetInstigator) false (Game|GetInstigator))
  (Utilities|FlowControl|Delay (Variables|Default|GetDelay_AfterMainDrop))
  (CallFunction|SpawnExtraDrops))

(event Custom|SpawnExtraDrops
  (bind _returnvalue (Game|GetInstigator))
  (bind _spawnedextracount (Variables|Default|GetSpawnedExtraCount))
  (bind _actorstodrop (Variables|Default|GetActorsToDrop))
  (bind _output (Utilities|Array|Get(acopy) _actorstodrop _spawnedextracount))
  ;[1 1 1] Loop through and drop the array of actors chosen.
  (Utilities|FlowControl|Delay (Variables|Default|GetDelay_BeforeExtraDrops))
  (Variables|Default|SetSpawnedExtraCount 0)
  (if (not (>= (Variables|Default|GetSpawnedExtraCount) (Utilities|Array|Length (Variables|Default|GetActorsToDrop))))
    (bind _output_get (Variables|Default|SetExtraSpawnTemp (CallFunction|MakeExtraSpawnerTransform)))
    (Ability|ExecuteGameplayCueWithParamsReplicated (Variables|Cues|GetGC_SpawnSecondary) (Ability|GameplayCue|MakeGameplayCueParameters :NormalizedMagnitude 0.0 :RawMagnitude 0.0 :Location (.location _output_get) :Normal "0,0,0" :GameplayEffectLevel 1 :AbilityLevel 1 :bReplicateLocationWhenUsingMinimalRepProxy false))
    (Utilities|Casting|CastToBGA_Athena_JollyRascal_ExtraDrop_ResourceClass _output
      (:then
        (CallFunction|SpawnResourceBox _output_get _returnvalue)
        (Math|Integer|IncrementInt (Variables|Default|GetSpawnedExtraCount)
          (:%20%20
            (Utilities|FlowControl|Delay (Math|Random|RandomFloatinRange (Variables|Default|GetDelay_ExtraSpawnMin) (Variables|Default|GetDelay_ExtraSpawnMax))))))
      (:CastFailed
        _))
    (Game|SpawnBuildingActorfromClass _output _output_get _returnvalue false _returnvalue))
  ;[1 1 1] Clean up and kill after drop
  (Utilities|FlowControl|Delay (Variables|Default|GetDelay_AfterDropsShrink))
  (CallFunction|CleanUpandKill))

(event Custom|SpawnInPortal
  (bind _riftfx_niagara (Variables|Default|GetRiftFX_Niagara))
  (bind _delaybeforedrop (Variables|Default|GetDelayBeforeDrop))
  ;[1 1 1] Spawn In Portal - play FX and scale up portal.
  (Components|Activation|Activate (Variables|Default|GetAmbientAudo) true)
  (Components|Activation|Activate _riftfx_niagara true)
  (Effects|Components|ParticleSystem|SetFloatParameter _riftfx_niagara "Lifetime" (* _delaybeforedrop 4.0))
  (Utilities|FlowControl|Delay _delaybeforedrop)
  (Ability|ExecuteGameplayCueReplicated (Variables|Cues|GetGC_Start))
  ;[1 1 1] Once the portal has scaled up do the drop
  (if (Networking|IsServer)
    (CallFunction|SpawnMainDrop)))

(event EventBeginPlay
  ;[1 1 1] Begin Play -
  (Development|PrintString "Portal Spawned")
  (Utilities|FlowControl|Delay (Variables|Default|GetDelayBeforePortal))
  (CallFunction|SpawninPortal)
  (if (Networking|IsServer)
    (CallFunction|SelectThingstoDrop)))

(event Custom|CleanUpAndKill
  ;[1 1 1] Clean Up And Kill - Shrink Down the portal, play FX, hide, and set to kill.
  (Components|Activation|Deactivate (Variables|Default|GetAmbientAudo))
  (Ability|ExecuteGameplayCueReplicated (Variables|Cues|GetGC_End))
  (bind (_direction _alpha) (Timeline|ScaleDownPortal PlayFromStart)
    (:Update
      (Transformation|SetRelativeScale3D (Variables|Default|GetMesh) (Math|Vector|Lerp(Vector) (Variables|Default|GetPortalLargeScale) (Variables|Default|GetPortalSmallScale) _alpha)))
    (:Finished
      (Rendering|SetActorHiddenInGame false)
      (Actor|SetLifeSpan (Variables|Default|GetLifespanAfterKill)))))

(fn ClampTransformToMinigameVolume
  (params (Transform :type Transform))
  (bind (_transform_location _transform_rotation _transform_scale) (Math|Transform|BreakTransform (Variables|GetTransform_Location)))
  (bind _outvolumemanager (Volume|GetVolumeManager)
    (:Valid
      (return (Math|Transform|MakeTransform (VolumeManager|GetNearestIslandVolumeforLocation _outvolumemanager _transform_location (Variables|Default|GetSpawnMinigameVolumePadding)) _transform_rotation _transform_scale)))
    (:Invalid
      (return (Variables|GetTransform)))))

(fn MakeExtraSpawnerTransform
  (bind (_returnvalue_x _returnvalue_y) (Math|Vector|BreakVector (* (Math|Rotator|GetRotationXVector (Math|Rotator|MakeRotator 0.0 0.0 (Math|Random|RandomFloatinRange 0.0 3.6e+02))) (LocalVariables|GetExtraDropSpawnRadius))))
  (bind (_returnvalue_x_2 _returnvalue_y_2 _returnvalue_z) (Math|Vector|BreakVector (Transformation|GetActorLocation)))
  (return (CallFunction|ClampTransformtoMinigameVolume (Math|Transform|MakeTransform (Math|Vector|MakeVector (+ _returnvalue_x_2 _returnvalue_x) (+ _returnvalue_y_2 _returnvalue_y) _returnvalue_z) (Math|Rotator|MakeRotator 0.0 0.0 (Math|Random|RandomFloatinRange 0.0 3.6e+02)) "1.000000,1.000000,1.000000"))))

(fn SelectThingsToDrop
  (bind _actorstodrop (Variables|Default|GetActorsToDrop))
  (for _ (range 1 (+ (Math|Random|RandomIntegerinRange (Math|Float|Truncate (ScalableFloat|GetValueatLevel (Variables|Rows|GetRow_MinExtraJunkToDrop))) (Math|Float|Truncate (ScalableFloat|GetValueatLevel (Variables|Rows|GetRow_MaxExtraJunkToDrop)))) 1))
    (Utilities|Array|Add _actorstodrop (LocalVariables|GetResourceBox)))
  (for _ (range 1 (+ (Math|Random|RandomIntegerinRange (Math|Float|Truncate (ScalableFloat|GetValueatLevel (Variables|Rows|GetRow_MinExtraBoxesToDrop))) (Math|Float|Truncate (ScalableFloat|GetValueatLevel (Variables|Rows|GetRow_MaxExtraBoxesToDrop)))) 1))
    (Utilities|Array|Add _actorstodrop (select (and (Math|Random|RandomBoolwithWeight (ScalableFloat|GetValueatLevel (Variables|Rows|GetRow_DroneBoxDropChance))) (not (Math|Float|NearlyEqual(Float) (ScalableFloat|GetValueatLevel (Variables|Rows|GetRow_GoldEnabled)) 1e+02 0.1))) (LocalVariables|GetDroneBox) (LocalVariables|GetAmmoBox))))
  (return))

(fn SpawnResourceBox
  (params (Transform :type Transform) (Instigator :type Pawn%20Object%20Reference))
  (bind _returnvalue (Math|Random|RandomFloatinRange 0.0 1.0))
  (bind _row_wooddropweight (Variables|Rows|GetRow_WoodDropWeight))
  (bind _returnvalue_2 (ScalableFloat|GetValueatLevel _row_wooddropweight))
  ;[1 1 1] Select the resource and spawn the actor. We do this here rather than in the spawned actor so that the server is spawning the actor with the resource type it is as to not have client server desync. We select both the resource type and a random mesh index to use on the actor.
  (if (<= _returnvalue_2 _returnvalue)
    (LocalVariables|SetResourceType "Wood")
    (elif (<= (+ _returnvalue_2 (ScalableFloat|GetValueatLevel (Variables|Rows|GetRow_StoneDropWeight))) _returnvalue)
      (LocalVariables|SetResourceType "Stone")
      (else
        (LocalVariables|SetResourceType "Metal"))))
  (Game|SpawnBuildingActorfromClass "/Game/Athena/Items/Consumables/JollyRascal/BGA_Athena_JollyRascal_ExtraDrop_Resource.BGA_Athena_JollyRascal_ExtraDrop_Resource_C" Transform Instigator false (LocalVariables|GetResourceType) (Math|Random|RandomIntegerinRange 0 1) 0 Instigator))

(fn ConstructionScript)
